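# Translucent UI material for the backing of an elevator panel entry, colored by state: time-restricted, restricted, complete, proceed or urgent.
# Nested static switches choose the color, with time-restricted drawing a two-tone TimerProgress fill.
# Restricted draws rotated, panning sine stripes in screen-space proportions, with alpha partly pulsed by layered time sines.
# Opacity combines the state alpha with a soft UV edge mask and a base-to-hover multiplier.
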Material /FortniteCoreUI/Frontend/Spatial/Art/Elevator/M_UI_ElevatorPanelEntry_Backing
graph {
 properties {
  bUsedWithStaticMesh=false
  blendMode=BLEND_Translucent
  materialDomain=MD_UI
  pixelDepthOffsetMode=PDOM_Legacy
  usedShadingModels=Default Lit
 }
 nodes [
  n1 = TextureCoordinate()
  n2 = Add(A=Add(A=OneMinus(None=@n1)) {constB=-1}, B=ScalarParameter() {defaultValue=0.333, parameterName=TimerProgress})
  n3 = Saturate(None=Multiply(A=ComponentMask(None=@n2) {r=true}, B=ScalarParameter() {defaultValue=64, parameterName=TimerProgress_BarSoftness}))
  n4 = LinearInterpolate(
   A=ScalarParameter() {defaultValue=0.25, group=Time, parameterName=Timer_NonProgress_Alpha}
   B=ScalarParameter() {defaultValue=0.5, group=Time, parameterName=Timer_Progress_Alpha}
   Alpha=@n3
  )
  n5 = MaterialFunctionCall!(
   Rotation_Fraction=ScalarParameter() {defaultValue=-0.05, group=Restricted, parameterName=AngledBars_Rotation}
   TexCoord=MaterialFunctionCall() {materialFunction=/FortUILibrary/Foundation/Art_ToolKit/Tools/MF_UI_SquareToScreenSpace}.X_Priority
  ) {materialFunction=/FortUILibrary/Foundation/Art_ToolKit/AngledBoxes/MF_UI_Rotator_Fractions}
  n6 = AppendVector(
   A=ScalarParameter() {defaultValue=-0.0125, group=Restricted, parameterName=X_PanSpeed}
   B=ScalarParameter() {group=Restricted, parameterName=Y_PanSpeed}
  )
  n7 = Sine(None=Multiply(A=Panner(Coordinate=@n5.Result, Speed=@n6), B=ScalarParameter() {defaultValue=32, group=Restricted, parameterName=AngledBars_Amount}))
  n8 = Saturate(None=Multiply(A=ComponentMask(None=@n7) {r=true}, B=ScalarParameter() {defaultValue=4, group=Restricted, parameterName=Bar_Softness}))
  n9 = ScalarParameter() {defaultValue=1, group=Restricted, parameterName=Bars_Alpha}
  n10 = Multiply(A=Time(), B=ScalarParameter() {defaultValue=0.6667, group=Restricted, parameterName=Overall_TimeMultiplier})
  n11 = LinearInterpolate(Alpha=Sine(None=@n10)) {constA=-3}
  n12 = Constant()
  n13 = Constant() {r=1}
  n14 = ScalarParameter() {defaultValue=3, group=Restricted, parameterName=Sine2_Time}
  n15 = Add(A=Sine(None=Multiply(A=@n10, B=@n14)))
  n16 = Saturate(None=DotProduct(A=If(A=@n11, B=@n12, "A > B"=@n13, "A == B"=@n13, "A < B"=@n12), B=LinearInterpolate(Alpha=@n15)))
  n17 = ScalarParameter() {defaultValue=0.5, group=Restricted, parameterName=Use_OscillatingTimeMultiplier}
  n18 = Add(
   A=Multiply(A=@n8, B=LinearInterpolate(A=@n9, B=@n16, Alpha=@n17))
   B=ScalarParameter() {defaultValue=0.5, group=Restricted, parameterName=Gaps_Alpha}
  )
  n19 = StaticSwitchParameter(
   True=Multiply(A=Saturate(None=@n18), B=ScalarParameter() {defaultValue=0.5, group=Restricted, parameterName=OverallRestrictedAlpha})
   False=Constant() {r=1}
  ) {group=StateSwitch, parameterName=IsRestricted}
  n20 = TextureCoordinate()
  n21 = ScalarParameter() {defaultValue=256, parameterName=RightLeftEdgeSoftness}
  n22 = Multiply(A=ComponentMask(None=@n20) {r=true}, B=@n21)
  n23 = Multiply(A=ComponentMask(None=OneMinus(None=@n20)) {r=true}, B=@n21)
  n24 = ScalarParameter() {defaultValue=64, parameterName=V_EdgeSoftness}
  n25 = Multiply(A=ComponentMask(None=@n20) {g=true}, B=@n24)
  n26 = Multiply(A=OneMinus(None=ComponentMask(None=@n20) {g=true}), B=@n24)
  n27 = DotProduct(A=DotProduct(A=Saturate(None=@n22), B=Saturate(None=@n23)), B=DotProduct(A=Saturate(None=@n25), B=Saturate(None=@n26)))
  Multiply(
   A=LinearInterpolate(B=StaticSwitchParameter(True=@n4, False=@n19) {parameterName=IsTimeRestricted}, Alpha=Saturate(None=@n27))
   B=LinearInterpolate(
    A=ScalarParameter() {defaultValue=0.75, group=Overall, parameterName=BaseAlphaState}
    Alpha=ScalarParameter() {group=Overall, parameterName=OnHover}
   ) {constA=0.5}
  )
  n28 = VectorParameter() {group=Time, parameterName=Timer_Color_1, defaultValue=<rgba 0.125 0.1875 0.5 1>}
  n29 = VectorParameter() {group=Time, parameterName=Timer_Color_2, defaultValue=<rgba 0.5 0.625 1 1>}
  n30 = LinearInterpolate(
   A=VectorParameter() {group=Restricted, parameterName=Restricted_Color_1, defaultValue=<rgba 0.5 0 0.0583 1>}.RGB
   B=VectorParameter() {group=Restricted, parameterName=Restricted_Color_2, defaultValue=<rgba 1 0 0.05 1>}.RGB
   Alpha=@n8
  )
  n31 = VectorParameter() {group=Urgent, parameterName=Urgent_Color, defaultValue=<rgba 0 0.2462 1 1>}
  n32 = Constant() {r=1}
  n33 = StaticSwitchParameter(
   True=VectorParameter() {group=Complete, parameterName=Complete_Color, defaultValue=<rgba 0 1 0.4167 1>}.RGB
   False=StaticSwitchParameter(
    True=VectorParameter() {group=Proceed, parameterName=Proceed_Color, defaultValue=<rgba 0 0.25 1 1>}.RGB
    False=StaticSwitchParameter(True=@n31.RGB, False=@n32) {parameterName=IsUrgent}
   ) {defaultValue=true, group=StateSwitch, parameterName=IsProceed}
  ) {group=StateSwitch, parameterName=IsComplete}
  Reroute(None=StaticSwitchParameter(
    True=LinearInterpolate(A=@n28.RGB, B=@n29.RGB, Alpha=@n3)
    False=StaticSwitchParameter(True=@n30, False=@n33) {group=StateSwitch, parameterName=IsRestricted}
   ) {group=StateSwitch, parameterName=IsTimeRestricted})
 ]
}
